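# Physics asset for the face accessory of the NanaEternal male medium head cosmetic, defining bodies and joints for its two dangling flaps.
# Each flap hangs from a small box body on its base bone.
# The right flap is a single capsule, while the left is a chain of two spheres.
# Angle-limited constraints join the segments, with tight secondary swing and twist, and wider primary swing, loosest at the left tip.
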PhysicsAsset /BRCosmetics/Characters/Player/Male/Medium/Heads/M_MED_NanaEternal_Head/Meshes/M_MED_NanaEternal_FaceAcc_Physics
bodies {
 dyn_flapBase_r [{shapeType=Box, extentX=1.01, extentY=1.01, extentZ=1.01, rotation=<rot 0 -90 0>}]
 dyn_flap_01_l [{shapeType=Sphere, radius=7, shapeName=dyn_flap_01_l_sphere, center=<5.4111 0 0>}]
 dyn_flap_02_l [{shapeType=Sphere, radius=6.5, shapeName=dyn_flap_02_l_sphere, center=<3.8973 0 0>}]
 dyn_flap_Base_l [{shapeType=Box, extentX=1.01, extentY=1.01, extentZ=1.01, rotation=<rot 0 -90 0>}]
 dyn_flap_r [{shapeType=Capsule, length=15, radius=6.5, shapeName=dyn_flap_r_capsule, center=<11.0648 0 0>, rotation=<rot 90 0 0>}]
}
constraints [
 {bone1Name=dyn_flap_r, bone2Name=dyn_flapBase_r, swing1LimitDegrees=30, swing2LimitDegrees=10, twistLimitDegrees=3}
 {bone1Name=dyn_flap_01_l, bone2Name=dyn_flap_Base_l, swing1LimitDegrees=30, swing2LimitDegrees=10, twistLimitDegrees=12}
 {bone1Name=dyn_flap_02_l, bone2Name=dyn_flap_01_l, swing1LimitDegrees=45, swing2LimitDegrees=10, twistLimitDegrees=12}
]
