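# Post-process Animation Blueprint for the NanaEternal male head cosmetic, adding a facial idle and secondary motion to accessory flaps.
# It copies the attached parent's pose and curves, then layers the Banana male facial idle on top per bone.
# Constraints drive the left and right flap base bones from spine_05 and neck_01, weighted toward the spine.
# A RigidBody node then simulates the face accessory physics asset in root-bone space, with damped component motion and overridden gravity.
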AnimBlueprint /BRCosmetics/Characters/Player/Male/Medium/Heads/M_MED_NanaEternal_Head/Meshes/CustomSkeleton/M_MED_NanaEternal_Head_PostProcess_AnimBP
target_skeleton=&/Fortnite_Base_Head_NanaEternal_Skeleton
graphs [
 AnimGraph {nodes [
   n1 = Copy Pose From Mesh() {bCopyCurves=true, bUseAttachedParent=true}
   n2 = Sequence Player 'Banana_Male_Facial_Idle'() {sequence=/Game/Animation/Game/MainPlayer/Facial/Banana/Banana_Male_Facial_Idle}
   n3 = "Constraint - dyn_flapBase_r (PP)"!(
    ComponentPose=Local To Component(LocalPose=Layered blend per bone(BasePose=@n1, BlendPoses_0=@n2))
    Alpha=1
    ConstraintWeights_0=0.8
    ConstraintWeights_1:float=0.2
   ) -> (Pose) {
    alphaBoolBlend {blendOption=Linear}
    boneToModify {boneName=dyn_flapBase_r}
    constraintSetup [{targetBone {boneName=spine_05}}, {targetBone {boneName=neck_01}}]
    constraintWeights [1,1]
   }
   n4 = "Constraint - dyn_flap_Base_l (PP)"!(ComponentPose=@n3, Alpha=1, ConstraintWeights_0=0.8, ConstraintWeights_1:float=0.2) -> (Pose) {
    alphaBoolBlend {blendOption=Linear}
    boneToModify {boneName=dyn_flap_Base_l}
    constraintSetup [{targetBone {boneName=spine_05}}, {targetBone {boneName=neck_01}}]
    constraintWeights [1,1]
   }
   n5 = RigidBody(ComponentPose=@n4) {
    bClampLinearTranslationLimitToRefPose=true
    bOverrideWorldGravity=true
    overridePhysicsAsset=/BRCosmetics/Characters/Player/Male/Medium/Heads/M_MED_NanaEternal_Head/Meshes/M_MED_NanaEternal_FaceAcc_Physics
    simulationSpace=BaseBoneSpace
    alphaBoolBlend {blendOption=Linear}
    baseBoneRef {boneName=root}
    componentAppliedLinearAccClamp=<800 800 800>
    componentLinearAccScale=<0.65 0.65 0.65>
    componentLinearVelScale=<0.3 0.3 0.3>
    overrideWorldGravity {z=-980}
   }
   Output Pose(Result=Component To Local(ComponentPose=@n5))
   Get DisableFaceOverride() -> (DisableFaceOverride:float)
  ]}
 EventGraph {nodes [TryGetPawnOwner()]}
]
variables []
